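# Additive, emissive-driven material for the Dragon SE vehicle wheel, passed through the shared damage and de-rez transition functions.
# Vertex color red separates rim from tire, blending their emissive and normals.
# A panning flow map distorts flame textures, tinted from Custom Color and masked for switchable rim and tire lights.
# Normal-based Fresnel terms, rim masks and a power-boosted tire texture build the final glow layers.
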Material /VehicleCosmetics/Wheels/Dragon_SE/Materials/M_Wheel_Dragon_SE
graph {
 properties {
  bAutomaticallySetUsageInEditor=false
  bUseMaterialAttributes=true
  bUsedWithMeshDeformer=true
  bUsedWithSkeletalMesh=true
  blendMode=BLEND_Additive
  usedShadingModels=Default Lit
 }
 nodes [
  n1 = LinearInterpolate(A=NamedRerouteUsage(), B=NamedRerouteUsage(), Alpha=VertexColor().R)
  MaterialFunctionCall(In=MaterialFunctionCall(Material=SetMaterialAttributes(Emissive Color=@n1) {attributeSetTypes [#1]}) {materialFunction=&MF_BaseVehicle_Damage}.Result) {materialFunction=&MF_VehicleDeRez_Transition}
  n2 = VectorParameter() {group=00 Base, parameterName=Custom Color, sortPriority=-3, defaultValue=<rgba 1 0.159 0 1>}
  NamedRerouteDeclaration(None=@n2.RGB) {name=Custom Color, nodeColor=<rgba 1 0.5 0 1>}
  n3 = TextureCoordinate() {uTiling=0.5, vTiling=0.3}
  n4 = ScalarParameter() {defaultValue=1, group=01 Tire Emissive, parameterName=Tire UV Distort Tile, sortPriority=1}
  n5 = TextureSample(UVs=Panner(Coordinate=Multiply(A=@n3, B=@n4)) {speedX=0.25, speedY=0.5}) {
   samplerSource=SSM_Wrap_WorldGroupSettings
   samplerType=SAMPLERTYPE_Normal
   texture=/VehicleCosmetics/Wheels/_SharedTextures/FX/Flow_03_nomips
  }
  n6 = ComponentMask(None=@n5.RGB) {g=true, r=true}
  n7 = ScalarParameter() {defaultValue=0.5, group=01 Tire Emissive, parameterName=Tire UV Distort Amt, sortPriority=2}
  n8 = TextureObjectParameter() {group=01 Tire Emissive, parameterName=Tire Flame, sortPriority=0, texture=/VehicleCosmetics/Wheels/_SharedTextures/FX/FlameTire_M_T}
  n9 = TextureSample(UVs=Add(A=TextureCoordinate(), B=Multiply(A=@n6, B=@n7)), Tex=@n8) {samplerSource=SSM_Wrap_WorldGroupSettings, texture=&FlameTire_M_T}
  n10 = Add(A=@n2.R, B=@n2.R)
  n11 = Clamp(None=Multiply(A=Add(A=@n10, B=@n2.B)) {constB=0.33})
  n12 = Constant()
  n13 = LinearInterpolate(
   A=If(A=@n11, "A > B"=@n2.RGB, "A == B"=@n12, "A < B"=Multiply(A=@n2.RGB) {constB=3}) {constB=0.15}
   B=If(A=@n11, "A > B"=@n2.RGB, "A == B"=@n12, "A < B"=Add(A=@n2.RGB) {constB=0.1}) {constB=0.5}
   Alpha=@n2.RGB
  )
  n14 = Multiply(A=Clamp(None=@n13), B=@n9.R)
  n15 = TextureObjectParameter() {
   group=00 Base
   parameterName=Rim Masks
   samplerType=SAMPLERTYPE_LinearColor
   sortPriority=2
   texture=/VehicleCosmetics/Wheels/Dragon_SE/Textures/T_Dragon_SE_M
  }
  n16 = TextureSample(Tex=@n15) {samplerSource=SSM_Wrap_WorldGroupSettings, texture=&FlameTire_M_T}
  n17 = Multiply(A=Multiply(A=Multiply(A=@n14) {constB=5}, B=@n16.R)) {constB=2}
  n18 = Constant()
  n19 = TextureCoordinate() {uTiling=0.5}
  n20 = TextureSample(UVs=@n19, Tex=@n8) {samplerSource=SSM_Wrap_WorldGroupSettings, texture=&FlameTire_M_T}
  n21 = Add(A=@n20.R, B=@n20.R)
  n22 = Clamp(None=OneMinus(None=@n21))
  n23 = Multiply(A=Multiply(A=Multiply(A=@n22, B=@n20.B), B=@n16.B), B=@n13)
  n24 = Add(A=@n16.B, B=@n20.B)
  n25 = Multiply(A=@n9.B, B=Clamp(None=Subtract(A=@n24, B=@n21)))
  n26 = Panner(Coordinate=@n19) {speedX=0.1}
  n27 = Multiply(A=Multiply(
    A=@n13
    B=TextureSample(UVs=@n26, Tex=@n8) {samplerSource=SSM_Wrap_WorldGroupSettings, texture=&FlameTire_M_T}.R
   )) {constB=3}
  n28 = Add(A=Add(A=@n14, B=Multiply(A=@n23) {constB=10}), B=Multiply(A=Multiply(A=@n25, B=@n27)) {constB=10})
  n29 = VertexColor()
  NamedRerouteDeclaration(None=LinearInterpolate(
    A=StaticSwitchParameter(True=@n17, False=@n18) {defaultValue=true, group=00 Base, parameterName=Rim Lights, sortPriority=-5}
    B=StaticSwitchParameter(True=@n28, False=@n18) {defaultValue=true, group=00 Base, parameterName=Tire Lights, sortPriority=-5}
    Alpha=@n29.R
   )) {name=Tire Emissive, nodeColor=<rgba 1 0 0.1641 1>}
  n30 = TextureObjectParameter() {
   group=00 Base
   parameterName=Rim Normal
   samplerType=SAMPLERTYPE_Normal
   sortPriority=1
   texture=/VehicleCosmetics/Wheels/Dragon_FNI/Textures/T_Dragon_Fni_Rim_N
  }
  n31 = TextureSample(Tex=@n30) {
   samplerSource=SSM_Wrap_WorldGroupSettings
   samplerType=SAMPLERTYPE_Normal
   texture=&T_Dragon_Fni_Rim_N
  }
  n32 = AppendVector(
   A=Multiply(
    A=ComponentMask(None=@n31.RGB) {g=true, r=true}
    B=ScalarParameter() {defaultValue=1, group=00 Base, parameterName=RIM Normal Intensity, sortPriority=0}
   )
   B=@n31.B
  )
  NamedRerouteDeclaration(None=@n32) {name=Rim Normal, nodeColor=<rgba 0 1 1 1>}
  NamedRerouteDeclaration(None=@n29.R) {name=VC_RC, nodeColor=<rgba 0.3594 0 1 1>}
  n33 = TextureObjectParameter() {
   group=00 Base
   parameterName=Tire Normal
   samplerType=SAMPLERTYPE_Normal
   sortPriority=1
   texture=/VehicleCosmetics/Wheels/Dragon_SE/Textures/T_Dragon_SE_Tire_N
  }
  n34 = TextureSample(Tex=@n33) {
   samplerSource=SSM_Wrap_WorldGroupSettings
   samplerType=SAMPLERTYPE_Normal
   texture=&T_Dragon_Fni_Rim_N
  }
  n35 = AppendVector(
   A=Multiply(
    A=ComponentMask(None=@n34.RGB) {g=true, r=true}
    B=ScalarParameter() {defaultValue=1, group=00 Base, parameterName=Tire Normal Intensity, sortPriority=0}
   )
   B=@n34.B
  )
  n36 = NamedRerouteUsage()
  n37 = DotProduct(
   A=Transform(None=CameraVectorWS()) {transformSourceType=TRANSFORMSOURCE_World, transformType=TRANSFORM_Tangent}
   B=NamedRerouteDeclaration(None=LinearInterpolate(A=@n32, B=@n35, Alpha=@n36)) {name=Normal, nodeColor=<rgba 0 1 1 1>}
  )
  n38 = Clamp(None=ConstantBiasScale(None=OneMinus(None=@n37)) {bias=-0.3, scale=1.5})
  n39 = Multiply(A=Clamp(None=Multiply(A=@n37) {constB=2}) {minDefault=0.35}, B=@n38)
  NamedRerouteDeclaration(None=Multiply(A=@n39, B=@n39)) {name=Fresnel, nodeColor=<rgba 0.1953 1 0 1>}
  n40 = Multiply(A=Power(Base=TextureSampleParameter2D() {group=00 Base, parameterName=Tire, texture=/VehicleCosmetics/Wheels/Dragon_SE/Textures/T_Wheel_Dragon_Tire_SE_RGB}.RGB) {constExponent=2.9}) {constB=12}
  n41 = Multiply(A=Multiply(A=NamedRerouteUsage()) {constB=0.5}, B=Add(A=NamedRerouteUsage(), B=Multiply(A=@n40, B=@n40) {constB=12}))
  NamedRerouteDeclaration(None=Add(A=Multiply(A=@n41) {constB=12}, B=NamedRerouteUsage())) {name=Tire Final Emissive, nodeColor=<rgba 1 0 0.5156 1>}
  n42 = Multiply(A=MaterialFunctionCall!(Normal, Exponent=Constant() {r=1}) {materialFunction=/VehicleCosmetics/Wheels/Cola_Holographic/Materials/MF_Fresnel_Classic}.Result) {constB=3}
  n43 = TextureSampleParameter2D() {
   group=00 Base
   parameterName=Rim Masks SE
   samplerType=SAMPLERTYPE_LinearColor
   texture=/VehicleCosmetics/Wheels/Dragon_SE/Textures/T_Dragon_SE_Rim_M
  }
  n44 = Multiply(A=Power(Base=@n43.A) {constExponent=4}) {constB=4}
  n45 = Multiply(A=@n44, B=@n44) {constB=4}
  n46 = Add(A=Add(A=Multiply(A=@n42, B=@n43.R) {constB=3}, B=Add(A=@n45, B=@n38)), B=Multiply(A=@n43.G) {constB=4})
  n47 = Multiply(
   A=@n43.B
   B=MaterialFunctionCall(Exponent=Constant() {r=0.5}) {materialFunction=&MF_Fresnel_Classic}.Result
  ) {constB=4}
  NamedRerouteDeclaration(None=Multiply(A=Add(A=@n46, B=@n47), B=NamedRerouteUsage())) {name=Rim Final Emissive, nodeColor=<rgba 0 0.7656 1 1>}
 ]
}
