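# Unlit translucent Niagara material for Dragon SE wheel VFX on sprites, ribbons and mesh particles, drawing a distorted smoke shape.
# A panning flow texture, tiled and scaled by dynamic parameters, warps smoke mask UVs, whose own panner has constant time.
# Clamp and subtract math carves the mask's red channel into a band, tinted by particle color and faded by vertex and particle alpha.
# DistortionAmount scales the raw mask separately; Brightness is unused.
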Material /VehicleCosmetics/Wheels/Dragon_SE/VFX/M_ParentNS_DragonSe
graph {
 properties {
  bAutomaticallySetUsageInEditor=false
  bUsedWithNiagaraMeshParticles=true
  bUsedWithNiagaraRibbons=true
  bUsedWithNiagaraSprites=true
  blendMode=BLEND_Translucent
  refractionMethod=RM_PixelNormalOffset
  shadingModel=MSM_Unlit
  usedShadingModels=Unlit
 }
 nodes [
  Constant() {r=3}
  n1 = DynamicParameter() {paramNames [DistortionAmount, Brightness, UVDistortTile, UVDistort]}
  n2 = Panner(Coordinate=Multiply(A=TextureCoordinate() {uTiling=0.3, vTiling=0.3}, B=@n1.UVDistortTile)) {speedX=0.25, speedY=0.5}
  n3 = Multiply(
   A=ComponentMask(None=TextureSample(UVs=@n2) {texture=/VehicleCosmetics/Wheels/DragonEndit/Textures/Flow_03}.RGB) {g=true, r=true}
   B=@n1.UVDistort
  )
  n4 = TextureCoordinate()
  n5 = TextureSample(UVs=Panner(Coordinate=Add(A=@n3, B=@n4), Time=Constant() {r=1}) {speedX=-0.4}) {texture=/VehicleCosmetics/Wheels/_SharedTextures/FX/SmokeMask_T}
  n6 = Multiply(A=@n5.R, B=Constant() {r=3})
  n7 = Clamp(None=Add(A=@n6, B=Constant() {r=-2}))
  n8 = Constant() {r=5}
  n9 = Clamp(None=Subtract(A=@n6, B=Multiply(A=@n7, B=@n8)))
  n10 = ParticleColor()
  Multiply(A=Multiply(A=@n9, B=@n10.RGB), B=Constant() {r=1})
  Multiply(A=Multiply(A=@n9, B=VertexColor().A), B=@n10.A)
  Multiply(A=@n1.DistortionAmount, B=@n5.RGB)
 ]
}
