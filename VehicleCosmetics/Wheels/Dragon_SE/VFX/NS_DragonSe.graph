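# Niagara system for the Dragon SE vehicle wheel cosmetic, emitting orange-red sprites that trail along -X and curl upward as they fade.
# One GPU emitter, named CompletelyEmpty despite its content, spawns particles at a steady rate in timed loops with a delay.
# Short-lived sprites spin, grow and lose alpha while velocity curves and random upward acceleration shape their path.
# Dynamic material parameters pass a per-particle random value to the sprite material.
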NiagaraSystem /VehicleCosmetics/Wheels/Dragon_SE/VFX/NS_DragonSe
dependencies {
 usedDataInterfaces [&NiagaraDataInterfaceCurve, &NiagaraDataInterfaceVectorCurve]
 usedDynamicInputs []
 usedModules [&SystemState]
}
emitters {CompletelyEmpty {
  simTarget=GPUComputeSim
  emitterUpdateScript [
   &EmitterState(Life Cycle Mode=::NewEnumerator1, Loop Delay=2, Loop Duration=6, UseLoopDelay=true)
   &SpawnRate(SpawnRate=36, Use Spawn Probability=false)
  ]
  particleSpawnScript [
   &InitializeParticle(
    Color=<rgba 1 0 0 1>
    Lifetime=0.8
    Sprite Rotation Mode=::NewEnumerator2
    Sprite Size Max=<20 25>
    Sprite Size Min=<10 10>
    Sprite Size Mode=::NewEnumerator2
    UsePositionOffset=false
    Write Material Random=true
   )
   &AddVelocity(Velocity=<-120 0 -32>)
   &ApplyInitialForces()
   &Color(Color=<rgba 1 0.16 0 1>)
   &DynamicMaterialParameters(
    Index 0 Param 1=1
    Index 0 Param 3=1
    Index 0 Param 4=&RandomRangeFloat(Minimum=0.5, Maximum=2)
    Module.Index0Float0Write=true
    Module.Index0Float1Write=true
    Module.Index0Float2Write=true
    Module.Index0Float3Write=true
   )
   &OffsetPosition(Position Offset=<0 0 -1.5>)
  ]
  particleUpdateScript [
   &ParticleState()
   &SpriteRotationRate(Rotation Rate=&RandomRangeFloat(Minimum=-72, Maximum=72))
   &ScaleSpriteSize(Uniform Curve Sprite Scale=DataInterface() {propertyValues {Curve {keys [{tangentMode=RCTM_User, time=0, value=0.5}, {tangentMode=RCTM_User, time=1, value=2}]}}})
   &ScaleVelocity(Velocity Scale=&VectorFromCurve(VectorCurve=DataInterface() {propertyValues {XCurve {keys [<at 0 0.6>,<at 1 1>]}, YCurve {keys [<at 0.009 0>,<at 1 1>]}, ZCurve {keys [<at 0 1>,<at 1 0>]}}}))
   &AccelerationForce(Acceleration=&RandomRangeVector(Minimum=<0 0 50>, Maximum=<0 0 100>))
   &SolveForcesAndVelocity(Clamp Velocity=false, Limit Acceleration=false)
   &ScaleColor(
    Scale Alpha=&FloatFromCurve(FloatCurve=DataInterface() {propertyValues {Curve {keys [<at 0 1>,<at 1 0>]}}})
    ScaleA=true
    ScaleRGB=true
   )
  ]
  rendererClasses [&NiagaraSpriteRendererProperties]
  renderers [{rendererClass=&NiagaraSpriteRendererProperties}]
 }}
summary {
 systemName=NS_DragonSe
 emitters [{emitterName=CompletelyEmpty, simTarget=GPUComputeSim, rendererClasses [&NiagaraSpriteRendererProperties]}]
 userVariables []
}
